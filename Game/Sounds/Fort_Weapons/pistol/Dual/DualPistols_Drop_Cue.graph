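# DualPistols_Drop_Cue is the sound cue for dropping the dual pistols.
# A Random node picks between two equally weighted drop waves.
# It plays through weapon-drop attenuation and the loot-collect sound class at reduced volume.
# It references a stop-oldest weapon-drop concurrency set and also carries single-instance, prevent-new override values.
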SoundCue /Game/Sounds/Fort_Weapons/pistol/Dual/DualPistols_Drop_Cue
root=Random_0
nodes [
 {id=Random_0, children [WavePlayer_1,WavePlayer_2], properties {weights [1,1]}}
 {id=WavePlayer_1, properties {soundWaveAssetPtr=&/DualPistols_Drop_01}}
 {id=WavePlayer_2, properties {soundWaveAssetPtr=&/DualPistols_Drop_02}}
]
properties {
 attenuationSettings=&Fort_Weapon_Drop_Att
 maxDistance=3000
 soundClassObject=&Fort_Loot_Collect
 volumeMultiplier=0.7
 concurrencyOverrides {maxCount=1, resolutionRule=PreventNew}
 concurrencySet [&Weapon_Drop_StopOldest]
}
