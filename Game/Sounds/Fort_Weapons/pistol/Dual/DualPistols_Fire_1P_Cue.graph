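# Sound cue for first-person dual pistol gunfire.
# A top random node picks evenly between two branches, each randomly choosing one of four identically named fire waves.
# One branch is pitched down to 0.8 and the other plays at normal pitch, giving two tonal variants of the shot.
# First-person weapon attenuation, sound class and gunfire concurrency govern playback, and the random nodes are excluded from branch culling.
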SoundCue /Game/Sounds/Fort_Weapons/pistol/Dual/DualPistols_Fire_1P_Cue
root=Modulator_2
nodes [
 {id=Modulator_0, children [Random_1], properties {pitchMax=0.8, pitchMin=0.8, volumeMax=1, volumeMin=1}}
 {id=Modulator_1, children [Random_2], properties {pitchMax=1, pitchMin=1, volumeMax=1, volumeMin=1}}
 {id=Modulator_2, children [Random_0], properties {pitchMax=1, pitchMin=1, volumeMax=1, volumeMin=1}}
 {id=Random_0, children [Modulator_0,Modulator_1], properties {bShouldExcludeFromBranchCulling=true, weights [1,1]}}
 {
  id=Random_1
  children [WavePlayer_4,WavePlayer_5,WavePlayer_6,WavePlayer_7]
  properties {bShouldExcludeFromBranchCulling=true, weights [1,1,1,1]}
 }
 {
  id=Random_2
  children [WavePlayer_0,WavePlayer_1,WavePlayer_2,WavePlayer_3]
  properties {bShouldExcludeFromBranchCulling=true, weights [1,1,1,1]}
 }
 {id=WavePlayer_0, properties {soundWaveAssetPtr=&/DualPistols_Fire_01}}
 {id=WavePlayer_1, properties {soundWaveAssetPtr=&/DualPistols_Fire_03}}
 {id=WavePlayer_2, properties {soundWaveAssetPtr=&/DualPistols_Fire_02}}
 {id=WavePlayer_3, properties {soundWaveAssetPtr=&/DualPistols_Fire_04}}
 {id=WavePlayer_4, properties {soundWaveAssetPtr=&DualPistols_Fire_01}}
 {id=WavePlayer_5, properties {soundWaveAssetPtr=&DualPistols_Fire_03}}
 {id=WavePlayer_6, properties {soundWaveAssetPtr=&DualPistols_Fire_02}}
 {id=WavePlayer_7, properties {soundWaveAssetPtr=&DualPistols_Fire_04}}
]
properties {
 attenuationSettings=&Fort_Weapon_Projectile_FirstPerson_Athena_Att
 maxDistance=2097152
 soundClassObject=&Fort_WeaponRanged_Fire_1P_PSM
 volumeMultiplier=0.65
 attenuationOverrides {distanceAlgorithm=NaturalSound, falloffDistance=5000, attenuationShapeExtents {x=1000}}
 concurrencySet [&Weapon_Gunfire_FirstPerson_Conc]
}
