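# Sound cue for dual pistols gunfire, routed to third-person and distant ranged-weapon fire sound classes.
# A mixer blends close, medium and distant layers at descending volumes, each with its own attenuation and sound class override.
# Each layer randomly picks a lowered or unity pitch branch, then a random wave; an outer modulator raises close-layer pitch or boosts medium and distant volume.
# Local and global gunfire concurrency sets limit playback.
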SoundCue /Game/Sounds/Fort_Weapons/pistol/Dual/DualPistols_Fire_Cue
root=Mixer_17
nodes [
 {
  id=Attenuation_0
  children [Modulator_0]
  properties {attenuationSettings=&Fort_Weapon_Projectile_Medium_Mono_Athena_Att}
 }
 {
  id=Attenuation_34
  children [Modulator_1]
  properties {attenuationSettings=&Fort_Weapon_Projectile_Distant_Mono_Athena_Att}
 }
 {id=Attenuation_49, children [Modulator_4], properties {attenuationSettings=&Fort_Weapon_Projectile_Close_Athena_Att}}
 {id=Mixer_17, children [SoundClass_32, SoundClass_0, SoundClass_35], properties {inputVolume [0.7,0.35,0.1]}}
 {id=Modulator_0, children [Random_8], properties {volumeMax=2, volumeMin=2}}
 {id=Modulator_1, children [Random_6], properties {volumeMax=2, volumeMin=2}}
 {id=Modulator_2, children [Random_3], properties {pitchMax=0.8, pitchMin=0.8, volumeMax=1, volumeMin=1}}
 {id=Modulator_3, children [Random_5], properties {pitchMax=1, pitchMin=1, volumeMax=1, volumeMin=1}}
 {id=Modulator_4, children [Random_4], properties {pitchMax=1.4, pitchMin=1.4, volumeMax=1, volumeMin=1}}
 {id=Modulator_5, children [Random_2], properties {pitchMax=0.8, pitchMin=0.8, volumeMax=1, volumeMin=1}}
 {id=Modulator_6, children [Random_7], properties {pitchMax=1, pitchMin=1, volumeMax=1, volumeMin=1}}
 {id=Modulator_8, children [Random_0], properties {pitchMax=0.8, pitchMin=0.8, volumeMax=1, volumeMin=1}}
 {id=Modulator_9, children [Random_9], properties {pitchMax=1, pitchMin=1, volumeMax=1, volumeMin=1}}
 {id=Random_0, children [WavePlayer_0,WavePlayer_1,WavePlayer_2], properties {bShouldExcludeFromBranchCulling=true, weights [1,1,1]}}
 {id=Random_2, children [WavePlayer_3,WavePlayer_8,WavePlayer_9], properties {bShouldExcludeFromBranchCulling=true, weights [1,1,1]}}
 {
  id=Random_3
  children [WavePlayer_10, WavePlayer_11, WavePlayer_12, WavePlayer_13]
  properties {bShouldExcludeFromBranchCulling=true, weights [1,1,1,1]}
 }
 {id=Random_4, children [Modulator_2,Modulator_3], properties {bShouldExcludeFromBranchCulling=true, weights [1,1]}}
 {
  id=Random_5
  children [WavePlayer_14, WavePlayer_15, WavePlayer_16, WavePlayer_17]
  properties {bShouldExcludeFromBranchCulling=true, weights [1,1,1,1]}
 }
 {id=Random_6, children [Modulator_5,Modulator_6], properties {bShouldExcludeFromBranchCulling=true, weights [1,1]}}
 {id=Random_7, children [WavePlayer_18, WavePlayer_19, WavePlayer_20], properties {bShouldExcludeFromBranchCulling=true, weights [1,1,1]}}
 {id=Random_8, children [Modulator_8,Modulator_9], properties {bShouldExcludeFromBranchCulling=true, weights [1,1]}}
 {id=Random_9, children [WavePlayer_21, WavePlayer_22, WavePlayer_23], properties {bShouldExcludeFromBranchCulling=true, weights [1,1,1]}}
 {
  id=SoundClass_0
  children [Attenuation_0]
  properties {soundClassOverride=&Fort_WeaponRanged_Fire_3P_PSM}
 }
 {
  id=SoundClass_32
  children [Attenuation_49]
  properties {soundClassOverride=&Fort_WeaponRanged_Fire_3P_PSM}
 }
 {
  id=SoundClass_35
  children [Attenuation_34]
  properties {soundClassOverride=&Fort_WeaponRanged_Fire_Distant}
 }
 {id=WavePlayer_0, properties {soundWaveAssetPtr=&/DualPistols_Fire_MediumNew_02}}
 {id=WavePlayer_1, properties {soundWaveAssetPtr=&/DualPistols_Fire_MediumNew_01}}
 {id=WavePlayer_10, properties {soundWaveAssetPtr=&/DualPistols_Fire_01}}
 {id=WavePlayer_11, properties {soundWaveAssetPtr=&/DualPistols_Fire_03}}
 {id=WavePlayer_12, properties {soundWaveAssetPtr=&/DualPistols_Fire_02}}
 {id=WavePlayer_13, properties {soundWaveAssetPtr=&/DualPistols_Fire_04}}
 {id=WavePlayer_14, properties {soundWaveAssetPtr=&DualPistols_Fire_01}}
 {id=WavePlayer_15, properties {soundWaveAssetPtr=&DualPistols_Fire_03}}
 {id=WavePlayer_16, properties {soundWaveAssetPtr=&DualPistols_Fire_02}}
 {id=WavePlayer_17, properties {soundWaveAssetPtr=&DualPistols_Fire_04}}
 {id=WavePlayer_18, properties {soundWaveAssetPtr=&/DualPistols_Fire_DistantNew_02}}
 {id=WavePlayer_19, properties {soundWaveAssetPtr=&/DualPistols_Fire_DistantNew_01}}
 {id=WavePlayer_2, properties {soundWaveAssetPtr=&/DualPistols_Fire_MediumNew_03}}
 {id=WavePlayer_20, properties {soundWaveAssetPtr=&/DualPistols_Fire_DistantNew_03}}
 {id=WavePlayer_21, properties {soundWaveAssetPtr=&DualPistols_Fire_MediumNew_02}}
 {id=WavePlayer_22, properties {soundWaveAssetPtr=&DualPistols_Fire_MediumNew_01}}
 {id=WavePlayer_23, properties {soundWaveAssetPtr=&DualPistols_Fire_MediumNew_03}}
 {id=WavePlayer_3, properties {soundWaveAssetPtr=&DualPistols_Fire_DistantNew_02}}
 {id=WavePlayer_8, properties {soundWaveAssetPtr=&DualPistols_Fire_DistantNew_01}}
 {id=WavePlayer_9, properties {soundWaveAssetPtr=&DualPistols_Fire_DistantNew_03}}
]
properties {
 maxDistance=27500
 volumeMultiplier=1
 attenuationOverrides {distanceAlgorithm=NaturalSound, falloffDistance=5000, attenuationShapeExtents {x=1000}}
 concurrencySet [&Weapon_Gunfire_Local_Conc, &Weapon_Gunfire_Global_Conc]
}
