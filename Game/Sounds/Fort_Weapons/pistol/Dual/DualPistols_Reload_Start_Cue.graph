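# Sound cue for the start of the dual pistols reload foley.
# A Random node picks one of three equally weighted reload-start waves and passes it through a Modulator at the root.
# It plays in the weapon foley sound class with the shared weapon reload attenuation.
# An overridden concurrency rule prevents new instances from starting once the limit is reached.
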SoundCue /Game/Sounds/Fort_Weapons/pistol/Dual/DualPistols_Reload_Start_Cue
root=Modulator_2
nodes [
 {id=Modulator_2, children [Random_0]}
 {id=Random_0, children [WavePlayer_0,WavePlayer_1,WavePlayer_2], properties {weights [1,1,1]}}
 {id=WavePlayer_0, properties {soundWaveAssetPtr=&/DualPistols_Reload_Start_01}}
 {id=WavePlayer_1, properties {soundWaveAssetPtr=&/DualPistols_Reload_Start_02}}
 {id=WavePlayer_2, properties {soundWaveAssetPtr=&/DualPistols_Reload_Start_03}}
]
properties {
 attenuationSettings=&Fort_Weapon_Reload_Att
 bOverrideConcurrency=true
 maxDistance=2250
 soundClassObject=&Fort_WeaponFoley
 volumeMultiplier=0.702
 attenuationOverrides {falloffDistance=4100, attenuationShapeExtents {x=900}}
 concurrencyOverrides {resolutionRule=PreventNew}
}
